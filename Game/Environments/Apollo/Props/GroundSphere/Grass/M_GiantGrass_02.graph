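# Two-sided foliage material for giant grass on Apollo, built for instanced meshes and Nanite.
# Base color blends per-instance tip tints with landscape virtual texture color sampled at object position, with Creative and no-RVT fallbacks.
# UV math builds normals and masks time-of-day subsurface color, while gameplay wind plus a camera-proximity HISM offset drive World Position Offset.
# MF_Curie_Landscape processes the attributes, alongside a flickering edge glow scaled by custom primitive data.
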Material /Game/Environments/Apollo/Props/GroundSphere/Grass/M_GiantGrass_02
graph {
 properties {
  bUseMaterialAttributes=true
  bUsedWithInstancedStaticMeshes=true
  bUsedWithNanite=true
  opacityMaskClipValue=0.333
  physMaterial=/Game/FortressPhysicalMaterials/Grass
  pixelDepthOffsetMode=PDOM_Legacy
  shadingModel=MSM_TwoSidedFoliage
  twoSided=true
  usedShadingModels=Two Sided Foliage
 }
 nodes [
  n1 = PerInstanceRandom()
  n2 = ScalarParameter()
  n3 = WorldPosition() {worldPositionShaderOffset=WPT_ExcludeAllShaderOffsets}
  n4 = ObjectPositionWS()
  n5 = RuntimeVirtualTextureSampleParameter(World Position=Add(A=@n3.XYZ, B=Subtract(A=@n4, B=@n3.XYZ))) {materialType=BaseColor_Normal_Specular_YCoCg, parameterName=RVT, virtualTexture=/Game/Athena/Apollo/Maps/Landscape/VT/RVT_Apollo_Landscape}
  n6 = ComponentMask(None=TextureCoordinate()) {g=true}
  n7 = LinearInterpolate(
   A=LinearInterpolate(
    A=VectorParameter() {parameterName=GrassTips A, defaultValue=<rgba 0.0432 0.1059 0.0061 1>}.RGB
    B=VectorParameter() {parameterName=GrassTips B, defaultValue=<rgba 0.2358 0.2517 0.0718 1>}.RGB
    Alpha=MaterialFunctionCall(In=@n1, Contrast=@n2) {materialFunction=&CheapContrast}.Result
   )
   B=VertexInterpolator(VS=@n5.BaseColor).PS
   Alpha=NamedRerouteDeclaration(None=@n6) {name=UV Grad, nodeColor=<rgba 0.5469 0 1 1>}
  )
  n8 = Constant() {r=1}
  n9 = Constant() {r=-2}
  n10 = VertexInterpolator(VS=Saturate(None=LinearInterpolate(A=@n8, B=@n9, Alpha=@n6)))
  LinearInterpolate(A=@n7, B=Multiply(A=@n7, B=ScalarParameter() {defaultValue=1.6, parameterName=TipBaseColorScale}), Alpha=@n10.PS)
  n11 = NamedRerouteUsage()
  n12 = SetMaterialAttributes!(
   MaterialAttributes
   Base Color=NamedRerouteUsage()
   Emissive Color=NamedRerouteUsage()
   Specular=LinearInterpolate(A=ScalarParameter() {defaultValue=0.25, parameterName=Specular}, B=NamedRerouteUsage(), Alpha=@n11)
   Roughness=LinearInterpolate(A=ScalarParameter() {defaultValue=0.25, parameterName=Roughness}, B=NamedRerouteUsage(), Alpha=@n11)
   Normal=NamedRerouteUsage()
   Opacity Mask
   World Position Offset=NamedRerouteUsage()
   Subsurface Color=NamedRerouteUsage()
  ) {attributeSetTypes [
    #1
    #2
    #3
    #4
    #5
    #6
    #7
    #8
   ]}
  MaterialFunctionCall!(
   In=@n12
   UseHeightFalloff
   OnFireEmissiveOverride
   HeightFalloffRadius=ScalarParameter() {defaultValue=150, parameterName=HeightFalloffRadius}
   HeightFalloffHardness=ScalarParameter() {defaultValue=0.2, parameterName=HeightFalloffHardness}
   VariationMasksMobile
   In_Grass
   Is_Stylised_WE
  ) {materialFunction=/Game/Packages/Fortress_SharedMaterials/Base_Material_Functions/MF_Curie_Landscape}
  n13 = Constant() {r=-0.3}
  n14 = ComponentMask(None=Add(
    A=MaterialFunctionCall(Texture Scale=@n13) {materialFunction=&ScaleUVsByCenter}.UVs
    B=Constant() {r=0.15}
   )) {r=true}
  n15 = AppendVector(A=@n14, B=OneMinus(None=@n14))
  n16 = DeriveNormalZ(InXY=Multiply(A=Subtract(A=@n15) {constB=0.5}) {constB=2})
  n17 = TwoSidedSign()
  n18 = Constant3Vector() {constant=<rgba 0 0 1 0>}
  n19 = LinearInterpolate(
   A=Transform(None=Multiply(A=@n17, B=@n18)) {transformSourceType=TRANSFORMSOURCE_Instance, transformType=TRANSFORM_Tangent}
   B=@n16
   Alpha=Constant() {r=0.3}
  )
  n20 = Constant3Vector() {constant=<rgba 0 0 1 0>}
  NamedRerouteDeclaration(None=StaticSwitchParameter(True=ShadingPathSwitch(Default=@n16, Forward=@n19, Mobile=@n19), False=QualitySwitch(Default=@n20, Epic=@n16)) {parameterName=ForCreative}) {name=Normals, nodeColor=<rgba 0 1 0.9766 1>}
  n21 = OneMinus(None=ComponentMask(None=TextureCoordinate()) {g=true})
  n22 = Constant() {r=2}
  n23 = MaterialFunctionCall!(SinkIntoGroundDistance, WindFalloffMask=Multiply(A=Power(Base=@n21, Exp=@n22), B=Constant() {r=0.7}), AfterHeightModification=@n23.Wind) {materialFunction=/Game/Athena/Apollo/Environments/Landscape/SmallFoliage/MasterMaterial/MaterialFunctions/MF_GrassWind-GameplayWPO}
  n24 = QualitySwitch(Default=Constant(), High=@n23.FinalWPO, Epic=@n23.FinalWPO)
  n25 = Multiply(A=NamedRerouteUsage(), B=VectorParameter() {parameterName=HISM_Offset, defaultValue=<rgba 0 0 -300 0>}.RGB)
  NamedRerouteDeclaration(None=StaticSwitchParameter(True=Add(A=@n24, B=@n25), False=Constant()) {defaultValue=true, parameterName=UseWPO}) {name=WPO, nodeColor=<rgba 0 0.6484 1 1>}
  n26 = VectorParameter() {parameterName=CreativeColor, defaultValue=<rgba 0.0574 0.12 0.012 0>}
  n27 = VectorParameter() {parameterName=CreativeColorTip, defaultValue=<rgba 0.0866 0.1233 0.0179 0>}
  n28 = LinearInterpolate(
   A=VectorParameter() {parameterName=MobileColor, defaultValue=<rgba 0.1054 0.1233 0.0124 0>}.RGB
   B=VectorParameter() {parameterName=MobileColorTip, defaultValue=<rgba 0.3594 0.3327 0.0955 0>}.RGB
   Alpha=@n10.PS
  )
  n29 = NamedRerouteDeclaration(None=StaticSwitchParameter(
    True=LinearInterpolate(A=@n26.RGB, B=@n27.RGB, Alpha=@n10.PS)
    False=MaterialFunctionCall!(NoRVT=@n28, RVT=@n7) {materialFunction=/Game/Packages/Fortress_SharedMaterials/Base_Material_Functions/FN_RVT_Switch_Color}.Result
   ) {parameterName=ForCreative}) {name=BaseColor, nodeColor=<rgba 0 0.8594 1 1>}
  n30 = MaterialFunctionCall(
   Color_Morning=VectorParameter() {parameterName=SSColor_Morning, defaultValue=<rgba 1 0.8509 0.5443 0>}.RGB
   Color_Day=VectorParameter() {parameterName=SSColor_Day, defaultValue=<rgba 2 2 2 0>}.RGB
   Color_Evening=VectorParameter() {parameterName=SSColor_Evening, defaultValue=<rgba 1.2 0.8197 0.1185 0>}.RGB
   Color_Night=VectorParameter() {parameterName=SSColor_Night, defaultValue=<rgba 1.5 1.5 1.5 0>}.RGB
  ) {materialFunction=&Color_Time_of_Day}
  n31 = SphereMask(A=ComponentMask(None=TextureCoordinate()) {r=true}, B=Constant() {r=0.525}, Radius=Constant() {r=0.13}, Hardness=Constant() {r=0.1})
  n32 = Power(Base=OneMinus(None=Multiply(A=@n31, B=@n31)), Exp=Constant() {r=2})
  n33 = Add(A=@n32, B=Constant() {r=0.6})
  NamedRerouteDeclaration(None=Multiply(A=Multiply(A=@n29, B=@n30.Result), B=Saturate(None=@n33))) {name=SSS Color, nodeColor=<rgba 1 0 0.7969 1>}
  n34 = ScalarParameter() {defaultValue=0.02, parameterName=EmissiveScale}
  NamedRerouteDeclaration(None=Add(A=Multiply(A=@n34, B=@n29), B=NamedRerouteUsage())) {name=Emissive, nodeColor=<rgba 0.4766 1 0 1>}
  n35 = Constant3Vector()
  NamedRerouteDeclaration(None=SphereMask(
    A=CameraPositionWS()
    B=TransformPosition(None=@n35) {transformSourceType=TRANSFORMPOSSOURCE_Instance, transformType=TRANSFORMPOSSOURCE_World}
    Radius=ScalarParameter() {defaultValue=350, parameterName=FadeRadius}
    Hardness=ScalarParameter() {defaultValue=0.1, parameterName=FadeHardness}
   )) {name=HISM_Mask, nodeColor=<rgba 0.875 0 1 1>}
  n36 = Fmod(
   A=Transform(None=ObjectPositionWS()) {transformSourceType=TRANSFORMSOURCE_Local, transformType=TRANSFORM_Instance}
   B=Constant3Vector() {constant=<rgba 0.888 0.724 0.334 0>}
  )
  n37 = Constant3Vector() {constant=<rgba 0.1441 0.35 0.884 0>}
  n38 = Sine(None=Multiply(A=Time(), B=DotProduct(A=@n36, B=@n37))) {period=0.25}
  n39 = Multiply(A=Add(A=@n32) {constB=0.12}, B=LinearInterpolate(Alpha=VertexInterpolator(VS=@n38).PS) {constA=0.6})
  n40 = VectorParameter() {parameterName=CurieEdge, defaultValue=<rgba 3 0.293 0.0469 1>}
  NamedRerouteDeclaration(None=Multiply(
    A=Multiply(A=@n39, B=@n40.RGB)
    B=ScalarParameter() {bUseCustomPrimitiveData=true, group=Curie, parameterName=CurieLERP, primitiveDataIndex=29}
   )) {name=CurieManualEdge, nodeColor=<rgba 1 0.2344 0 1>}
  NamedRerouteDeclaration(None=@n5.Roughness) {name=Roughness RVT, nodeColor=<rgba 0 1 0.2969 1>}
  NamedRerouteDeclaration(None=@n5.Specular) {name=Spec RVT, nodeColor=<rgba 0 1 0.2969 1>}
  Reroute(None=@n6)
  ScalarParameter() {defaultValue=25, parameterName=RVT_MIP}
 ]
}
